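# Sound cue for a bush prop being destroyed, using prop-search attenuation and a concurrency override that prevents new instances at the limit.
# A quality-level node selects among four branches.
# The first two mix random bush-search rustles, pitched up and faded by an enveloper, with half-volume random hay hits, the first having more variations.
# The last two each play a single optimized destroyed wave with slight pitch variation.
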SoundCue /Game/Sounds/Foley_Loot/Containers/Bushes/Prop_Bush_Destroyed_Cue
root=QualityLevel_0
nodes [
 {
  id=Enveloper_0
  children [Random_1]
  properties {pitchCurve {editorCurveData {keys [{time=0}]}}, volumeCurve {editorCurveData {keys [<at 0 1>,<at 1 0>]}}}
 }
 {
  id=Enveloper_1
  children [Random_0]
  properties {pitchCurve {editorCurveData {keys [{time=0}]}}, volumeCurve {editorCurveData {keys [<at 0 1>,<at 1 0>]}}}
 }
 {id=Mixer_0, children [Modulator_0,Modulator_2], properties {inputVolume [1,0.5]}}
 {id=Mixer_1, children [Modulator_1,Modulator_5], properties {inputVolume [1,0.5]}}
 {id=Modulator_0, children [Enveloper_1], properties {pitchMax=1.3, pitchMin=1}}
 {id=Modulator_1, children [Enveloper_0], properties {pitchMax=1.3, pitchMin=1}}
 {id=Modulator_2, children [Random_2], properties {pitchMax=1, pitchMin=1}}
 {id=Modulator_3, children [WavePlayer_0], properties {pitchMax=1.08, pitchMin=0.92}}
 {id=Modulator_4, children [WavePlayer_10], properties {pitchMax=1.08, pitchMin=0.92}}
 {id=Modulator_5, children [Random_6], properties {pitchMax=1, pitchMin=1}}
 {id=QualityLevel_0, children [Mixer_1,Mixer_0,Modulator_3,Modulator_4]}
 {id=Random_0, children [WavePlayer_1,WavePlayer_2], properties {weights [1,1]}}
 {id=Random_1, children [WavePlayer_6,WavePlayer_7,WavePlayer_8], properties {weights [1,1,1]}}
 {id=Random_2, children [WavePlayer_4,WavePlayer_5], properties {weights [1,1]}}
 {id=Random_6, children [WavePlayer_24, WavePlayer_26, WavePlayer_27], properties {weights [1,1,1]}}
 {id=WavePlayer_0, properties {soundWaveAssetPtr=&/Bush_Detroyed_opt}}
 {id=WavePlayer_1, properties {soundWaveAssetPtr=&/Bush_Search_ST_01}}
 {id=WavePlayer_10, properties {soundWaveAssetPtr=&Bush_Detroyed_opt}}
 {id=WavePlayer_2, properties {soundWaveAssetPtr=&/Bush_Search_ST_02}}
 {id=WavePlayer_24, properties {soundWaveAssetPtr=/Game/Sounds/Fort_Impact_Sounds/Demolition/Hay/hay_hit_04}}
 {id=WavePlayer_26, properties {soundWaveAssetPtr=/Game/Sounds/Fort_Impact_Sounds/Demolition/Hay/hay_hit_02}}
 {id=WavePlayer_27, properties {soundWaveAssetPtr=/Game/Sounds/Fort_Impact_Sounds/Demolition/Hay/hay_hit_03}}
 {id=WavePlayer_4, properties {soundWaveAssetPtr=&hay_hit_04}}
 {id=WavePlayer_5, properties {soundWaveAssetPtr=&hay_hit_02}}
 {id=WavePlayer_6, properties {soundWaveAssetPtr=&Bush_Search_ST_01}}
 {id=WavePlayer_7, properties {soundWaveAssetPtr=&Bush_Search_ST_02}}
 {id=WavePlayer_8, properties {soundWaveAssetPtr=&/Bush_Search_ST_03}}
]
properties {
 attenuationSettings=/Game/Sounds/Attenuation/Fort_Prop_Search_Att
 bOverrideConcurrency=true
 maxDistance=1300
 soundClassObject=&Fort_Loot_Collect
 volumeMultiplier=1
 concurrencyOverrides {resolutionRule=PreventNew}
}
