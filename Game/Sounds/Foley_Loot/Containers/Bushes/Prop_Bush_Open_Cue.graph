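# Sound cue for opening a searchable bush container, part of the loot foley set.
# A Random node picks one of three equally weighted bush-searching open waves and is excluded from branch culling, feeding a Modulator that varies pitch.
# It uses the item-interact attenuation and loot-collect sound class, with a slight volume boost.
# A concurrency override prevents new instances from starting once the limit is reached.
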SoundCue /Game/Sounds/Foley_Loot/Containers/Bushes/Prop_Bush_Open_Cue
root=Modulator_0
nodes [
 {id=Modulator_0, children [Random_0], properties {pitchMax=1.2}}
 {id=Random_0, children [WavePlayer_0,WavePlayer_1,WavePlayer_2], properties {bShouldExcludeFromBranchCulling=true, weights [1,1,1]}}
 {id=WavePlayer_0, properties {soundWaveAssetPtr=&/Bush_Searching_Open_01}}
 {id=WavePlayer_1, properties {soundWaveAssetPtr=&/Bush_Searching_Open_02}}
 {id=WavePlayer_2, properties {soundWaveAssetPtr=&/Bush_Searching_Open_03}}
]
properties {
 attenuationSettings=/Game/Sounds/Attenuation/Fort_Item_Interact_Att
 bOverrideConcurrency=true
 maxDistance=1850
 soundClassObject=&Fort_Loot_Collect
 volumeMultiplier=1.15
 concurrencyOverrides {resolutionRule=PreventNew}
}
