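# Sound cue for searching a bush loot container, layering a short rustle over a sustained search loop.
# One of four short rustles plays with random pitch and volume, its envelope dropping to half volume.
# Mixed much lower, the long loop fades in and gets its own pitch modulation.
# It uses item-interact attenuation and the loot-collect sound class, with a concurrency override that blocks new instances at the limit.
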SoundCue /Game/Sounds/Foley_Loot/Containers/Bushes/Prop_Bush_Searching_Long_Loop_Cue
root=Mixer_0
nodes [
 {
  id=Enveloper_0
  children [WavePlayer_7]
  properties {pitchCurve {editorCurveData {keys [{time=0}]}}, volumeCurve {editorCurveData {keys [<at 0 0 RCIM_Cubic>,<at 1.5 1>]}}}
 }
 {
  id=Enveloper_1
  children [Modulator_1]
  properties {pitchCurve {editorCurveData {keys [{time=0}]}}, volumeCurve {editorCurveData {keys [<at 0 1>,<at 0.5 0.5>]}}}
 }
 {id=Mixer_0, children [Enveloper_1,Modulator_0], properties {inputVolume [0.9,0.25]}}
 {id=Modulator_0, children [Enveloper_0], properties {pitchMax=1.2}}
 {id=Modulator_1, children [Random_0], properties {pitchMax=1.5, volumeMax=1, volumeMin=0.65}}
 {
  id=Random_0
  children [WavePlayer_0,WavePlayer_1,WavePlayer_2,WavePlayer_3]
  properties {bShouldExcludeFromBranchCulling=true, weights [1,1,1,1]}
 }
 {id=WavePlayer_0, properties {soundWaveAssetPtr=&/Bush_Search_Short_01}}
 {id=WavePlayer_1, properties {soundWaveAssetPtr=&/Bush_Search_Short_02}}
 {id=WavePlayer_2, properties {soundWaveAssetPtr=&/Bush_Search_Short_03}}
 {id=WavePlayer_3, properties {soundWaveAssetPtr=&/Bush_Search_Short_04}}
 {id=WavePlayer_7, properties {bLooping=true, soundWaveAssetPtr=&/Bush_Search_Long_Loop}}
]
properties {
 attenuationSettings=/Game/Sounds/Attenuation/Fort_Item_Interact_Mono_Att
 bDebug=true
 bOverrideConcurrency=true
 maxDistance=1500
 soundClassObject=&Fort_Loot_Collect
 volumeMultiplier=2.1188
 concurrencyOverrides {resolutionRule=PreventNew}
}
