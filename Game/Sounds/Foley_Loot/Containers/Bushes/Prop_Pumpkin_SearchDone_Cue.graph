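# Sound cue for finishing a pumpkin container search, pairing the no-more-loot UI sound with a short layer of search foley.
# The foley mixes a quieter, pitch-raised random bush rustle with a random gear search clip, and an envelope fades both out in under a second.
# It uses item-interact attenuation and the loot-collect sound class, with a concurrency override that blocks a second instance while one plays.
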SoundCue /Game/Sounds/Foley_Loot/Containers/Bushes/Prop_Pumpkin_SearchDone_Cue
root=Mixer_1
nodes [
 {
  id=Enveloper_0
  children [Mixer_7]
  properties {pitchCurve {editorCurveData {keys [{time=0}]}}, volumeCurve {editorCurveData {keys [<at 0 1>,<at 0.7 0>]}}}
 }
 {id=Mixer_1, children [WavePlayer_12, Enveloper_0], properties {inputVolume [0.8,1]}}
 {id=Mixer_7, children [Modulator_12,Random_23], properties {inputVolume [0.3,1]}}
 {id=Modulator_12, children [Random_2], properties {pitchMax=1.25, pitchMin=1}}
 {
  id=Random_2
  children [WavePlayer_11, WavePlayer_13, WavePlayer_14, WavePlayer_9]
  properties {bShouldExcludeFromBranchCulling=true, weights [1,1,1,1]}
 }
 {id=Random_23, children [WavePlayer_137, WavePlayer_141, WavePlayer_142], properties {weights [1,1,1]}}
 {id=WavePlayer_11, properties {soundWaveAssetPtr=&/Bush_Search_Short_01}}
 {id=WavePlayer_12, properties {soundWaveAssetPtr=/Game/Sounds/Fort_UI_Sounds/Gameplay/Fort_UI_NoMoreLoot}}
 {id=WavePlayer_13, properties {soundWaveAssetPtr=&/Bush_Search_Short_02}}
 {id=WavePlayer_137, properties {soundWaveAssetPtr=/Game/Sounds/Foley_Loot/Containers/Misc/Prop_Gear_Search_01}}
 {id=WavePlayer_14, properties {soundWaveAssetPtr=&/Bush_Search_Short_03}}
 {id=WavePlayer_141, properties {soundWaveAssetPtr=/Game/Sounds/Foley_Loot/Containers/Misc/Prop_Gear_Search_05}}
 {id=WavePlayer_142, properties {soundWaveAssetPtr=/Game/Sounds/Foley_Loot/Containers/Misc/Prop_Gear_Search_06}}
 {id=WavePlayer_9, properties {soundWaveAssetPtr=&/Bush_Search_Short_04}}
]
properties {
 attenuationSettings=/Game/Sounds/Attenuation/Fort_Item_Interact_Mono_Att
 bOverrideConcurrency=true
 maxDistance=1500
 soundClassObject=&Fort_Loot_Collect
 volumeMultiplier=0.7
 concurrencyOverrides {maxCount=1, resolutionRule=PreventNew}
}
