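# Sound cue for searching a pumpkin prop loot container, built from bush and gear search foley.
# A mixer layers a random short bush rustle, pitch-modulated upward, with a slightly louder random gear search sound at its original pitch.
# It uses item-interact mono attenuation and the loot-collect sound class, with a concurrency override that prevents new instances at the limit.
# The cue's debug flag is left on.
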SoundCue /Game/Sounds/Foley_Loot/Containers/Bushes/Prop_Pumpkin_Search_Cue
root=Mixer_6
nodes [
 {id=Mixer_6, children [Modulator_12,Random_21], properties {inputVolume [0.5,0.7]}}
 {id=Modulator_12, children [Random_2], properties {pitchMax=1.25, pitchMin=1}}
 {
  id=Random_2
  children [WavePlayer_3,WavePlayer_6,WavePlayer_8,WavePlayer_9]
  properties {bShouldExcludeFromBranchCulling=true, weights [1,1,1,1]}
 }
 {id=Random_21, children [WavePlayer_131, WavePlayer_132, WavePlayer_127], properties {weights [1,1,1]}}
 {id=WavePlayer_127, properties {soundWaveAssetPtr=/Game/Sounds/Foley_Loot/Containers/Misc/Prop_Gear_Search_01}}
 {id=WavePlayer_131, properties {soundWaveAssetPtr=/Game/Sounds/Foley_Loot/Containers/Misc/Prop_Gear_Search_05}}
 {id=WavePlayer_132, properties {soundWaveAssetPtr=/Game/Sounds/Foley_Loot/Containers/Misc/Prop_Gear_Search_06}}
 {id=WavePlayer_3, properties {soundWaveAssetPtr=&/Bush_Search_Short_01}}
 {id=WavePlayer_6, properties {soundWaveAssetPtr=&/Bush_Search_Short_02}}
 {id=WavePlayer_8, properties {soundWaveAssetPtr=&/Bush_Search_Short_03}}
 {id=WavePlayer_9, properties {soundWaveAssetPtr=&/Bush_Search_Short_04}}
]
properties {
 attenuationSettings=/Game/Sounds/Attenuation/Fort_Item_Interact_Mono_Att
 bDebug=true
 bOverrideConcurrency=true
 maxDistance=1500
 soundClassObject=&Fort_Loot_Collect
 volumeMultiplier=1.3
 concurrencyOverrides {resolutionRule=PreventNew}
}
